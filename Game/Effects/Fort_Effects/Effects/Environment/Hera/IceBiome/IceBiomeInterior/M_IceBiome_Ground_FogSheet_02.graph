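# Unlit additive material for mesh-particle ground fog sheets in the Ice Biome interior.
# Layers panning cloud noise, with a low-frequency sample for distortion and actor position varying timing and channel blending per placement.
# Opacity combines noise, soft UV edge masks, view-angle, camera and depth fades, alongside a flat color and normal-direction vertex displacement.
# Quality and shading-path switches drop depth fade and displacement on low, medium, forward and mobile.
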Material /Game/Effects/Fort_Effects/Effects/Environment/Hera/IceBiome/IceBiomeInterior/M_IceBiome_Ground_FogSheet_02
graph {
 properties {
  bUsedWithMeshParticles=true
  bUsedWithNiagaraMeshParticles=true
  blendMode=BLEND_Additive
  pixelDepthOffsetMode=PDOM_Legacy
  shadingModel=MSM_Unlit
  usedShadingModels=Unlit
 }
 nodes [
  n1 = ActorPositionWS()
  n2 = ComponentMask(None=@n1) {r=true}
  Frac(None=@n2)
  n3 = ComponentMask(None=@n1) {g=true}
  Frac(None=@n3)
  n4 = Divide(A=MaterialFunctionCall(f2=ActorPositionWS()) {materialFunction=&AddComponents}.f2) {constB=100}
  n5 = Time()
  NamedRerouteDeclaration(None=Multiply(A=Add(A=@n4, B=@n5), B=ScalarParameter() {defaultValue=1.2, group=Panning, parameterName=Global Panner Speed, sortPriority=-1})) {name=Time Settings, nodeColor=<rgba 0 0.5078 1 1>}
  n6 = TextureCoordinate()
  n7 = Constant() {r=0.5}
  n8 = Multiply(
   A=SphereMask(A=ComponentMask(None=@n6) {g=true}, B=@n7, Radius=@n7) {hardnessPercent=0}
   B=SphereMask(A=ComponentMask(None=@n6) {r=true}, B=@n7, Radius=@n7) {hardnessPercent=0}
  )
  n9 = ScalarParameter() {defaultValue=0.9, group=Texture, parameterName=U Lo Freq, sortPriority=1}
  n10 = ScalarParameter() {defaultValue=1, group=Texture, parameterName=V Lo Freq, sortPriority=2}
  n11 = Multiply(A=TextureCoordinate(), B=ScalarParameter() {defaultValue=1, group=Texture, parameterName=Global UV Scalar, sortPriority=0})
  n12 = Panner(
   Coordinate=Multiply(A=@n11, B=AppendVector(A=@n9, B=@n10))
   Time=NamedRerouteUsage()
   Speed=AppendVector(A=Constant(), B=ScalarParameter() {defaultValue=-1, group=Panning, parameterName=Panner Speed Lo Freq, sortPriority=1})
  ) {bFractionalPart=true, speedY=-1.5}
  n13 = AppendVector(
   A=NamedRerouteDeclaration(None=TextureSample(UVs=@n12) {samplerType=SAMPLERTYPE_LinearGrayscale, texture=/Game/Global/Textures/Default/Tileable_Patterns/T_StormCloudDetail01_M}.G) {name=Lo Freq Noise, nodeColor=<rgba 0 0.1094 1 1>}
   B=Constant()
  )
  n14 = Subtract(A=Multiply(A=Saturate(None=@n13)) {constB=2})
  n15 = Multiply(A=ScalarParameter() {group=Texture, parameterName=Distortion Amount, sortPriority=5}, B=OneMinus(None=NamedRerouteUsage()))
  n16 = Multiply(
   A=@n11
   B=AppendVector(
    A=ScalarParameter() {defaultValue=1.32, group=Texture, parameterName=U Med Freq, sortPriority=10}
    B=ScalarParameter() {defaultValue=1.4, group=Texture, parameterName=V Med Freq, sortPriority=11}
   )
  )
  n17 = NamedRerouteUsage()
  n18 = ScalarParameter() {defaultValue=-1, group=Panning, parameterName=Panner Speed Med Freq, sortPriority=2}
  n19 = AppendVector(A=Constant(), B=@n18)
  n20 = TextureObject() {texture=/Game/Global/Textures/Default/Noise/T_Noise_Cloud}
  n21 = TextureSample(UVs=Add(A=Multiply(A=@n14, B=@n15), B=Panner(Coordinate=@n16, Time=@n17, Speed=@n19) {bFractionalPart=true, speedY=-1}), Tex=@n20) {texture=/Game/Athena/SafeZone/S11/T_TilingClouds_D}
  n22 = Panner(Coordinate=@n16, Time=NamedRerouteUsage(), Speed=AppendVector(A=Constant(), B=Multiply(A=@n18) {constB=2})) {bFractionalPart=true, speedY=-1}
  n23 = TextureSample(UVs=Add(A=@n22, B=Multiply(A=@n22) {constB=2}), Tex=@n20) {texture=&T_TilingClouds_D}
  n24 = Add(
   A=MaterialFunctionCall(1=@n21.R, 2=@n21.R, 3=@n21.R, A=@n2) {materialFunction=&Lerp_Multiple_Float3}.Lerp 3 Inputs
   B=Multiply(A=MaterialFunctionCall(1=@n23.R, 2=@n23.R, 3=@n23.R, A=@n3) {materialFunction=&Lerp_Multiple_Float3}.Lerp 3 Inputs) {constB=0.5}
  )
  n25 = NamedRerouteUsage()
  n26 = NamedRerouteDeclaration(None=Multiply(A=Saturate(None=@n24), B=Power(Base=@n25) {constExponent=3})) {name=Final Noise, nodeColor=<rgba 1 0.2109 0 1>}
  n27 = ScalarParameter() {defaultValue=0.4, group=Opacity, parameterName=AngleFadeSensitivity}
  n28 = Multiply(
   A=Multiply(
    A=Multiply(A=@n8, B=@n26)
    B=MaterialFunctionCall("FadeSensitivity (optional)"=@n27) {materialFunction=&FadeBasedOnViewAngle}.Result
   )
   B=MaterialFunctionCall(
    Fade Length=ScalarParameter() {defaultValue=100, group=Opacity, parameterName=Cam Depth Fade Length}
    Fade Offset=ScalarParameter() {group=Opacity, parameterName=Cam Depth Fade Offset}
    For Vertex Shader=StaticBoolParameter() {defaultValue=true, group=Opacity, parameterName="For Vertex Shader (SB)"}
   ) {materialFunction=&CameraDepthFade}.Result
  )
  n29 = DepthFade(
   Opacity=Multiply(A=@n28, B=ScalarParameter() {defaultValue=1, group=Opacity, parameterName=Opacity Mult, sortPriority=-2})
   FadeDistance=ScalarParameter() {defaultValue=100, group=Opacity, parameterName=Depth Fade Distance, sortPriority=6}
  )
  n30 = Multiply(A=@n28, B=ScalarParameter() {defaultValue=0.5, group=Opacity, parameterName=Opacity Mult Low End, sortPriority=-1})
  n31 = QualitySwitch(Default=@n29, Low=@n30, High=@n29, Medium=@n30, Epic=@n29)
  NamedRerouteDeclaration(None=Saturate(None=ShadingPathSwitch(Default=@n31, Deferred=@n31, Forward=@n30, Mobile=@n30))) {name=Final Opacity, nodeColor=<rgba 0.4297 0 1 1>}
  NamedRerouteDeclaration(None=VectorParameter() {group=Color, parameterName=Color, sortPriority=0, defaultValue=<rgba 1 1 1 0>}.RGB) {name=Final Emissive, nodeColor=<rgba 0 1 0.1094 1>}
  n32 = Multiply(A=VertexNormalWS(), B=Multiply(A=NamedRerouteUsage(), B=ScalarParameter() {defaultValue=1, parameterName=Displacement Amount}))
  n33 = Constant()
  n34 = QualitySwitch(Default=@n32, Low=@n33, High=@n32, Medium=@n33, Epic=@n32)
  NamedRerouteDeclaration(None=ShadingPathSwitch(Default=@n34, Deferred=@n34, Forward=@n33, Mobile=@n33)) {name=Final WPO, nodeColor=<rgba 1 0 0.2109 1>}
  n35 = OneMinus(None=ComponentMask(None=TextureCoordinate()) {g=true})
  NamedRerouteDeclaration(None=Saturate(None=Power(Base=@n35))) {name=V Mask, nodeColor=<rgba 1 0.0469 0 1>}
  NamedRerouteUsage()
  NamedRerouteUsage()
  NamedRerouteUsage()
 ]
}
