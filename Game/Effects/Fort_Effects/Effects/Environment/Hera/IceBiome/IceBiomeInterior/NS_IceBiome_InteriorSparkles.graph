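# Niagara system for ice biome interior sparkles, with one GPU sprite emitter that is disabled in this asset.
# Particles spawn in a box, then GBuffer sampling sets depth and color, with culling by view mask, surface normal and roughness.
# Updates apply camera-distance scaling, an alpha curve, wind and drag, and kill particles outside a volume.
# A rapidly rotating random glint vector and brightened color are written to particle attributes for two sprite renderers.
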NiagaraSystem /Game/Effects/Fort_Effects/Effects/Environment/Hera/IceBiome/IceBiomeInterior/NS_IceBiome_InteriorSparkles
dependencies {
 usedDataInterfaces [
  &NiagaraDataInterfaceGBuffer
  &NiagaraDataInterfaceCurve
  &NiagaraDataInterfaceCamera
  &NiagaraDataInterfaceCollisionQuery
  &NiagaraDataInterfaceLandscape
 ]
 usedDynamicInputs []
 usedModules [&SystemState]
}
emitters {InteriorSparkles {
  bEnabled=false
  simTarget=GPUComputeSim
  emitterUpdateScript [
   &EmitterState(
    Life Cycle Mode=::NewEnumerator1
    Loop Duration=&RandomRangeFloat(Minimum=2, Maximum=3)
    Recalculate Duration Each Loop=true
    Scalability Mode=::NewEnumerator1
    UseLoopDelay=false
   )
   &SpawnRate(SpawnRate=500, Use Spawn Probability=false)
  ]
  particleSpawnScript [
   &InitializeParticle(
    Color Mode=::NewEnumerator0
    Lifetime Max=4
    Lifetime Min=2
    Lifetime Mode=::NewEnumerator1
    Sprite Size Mode=::NewEnumerator1
    Uniform Sprite Size Max=20
    Uniform Sprite Size Min=20
    UsePositionOffset=false
    Write Material Random=true
   ) {module=InitializeParticle001}
   &ShapeLocation(
    Box Size=<1100 800 1500>
    MaskSpawnGroup=false
    Offset Mode=::NewEnumerator0
    Override Seed=false
    Shape Primitive=::NewEnumerator2
    Surface Only=false
    UseOwnerScale=true
   )
   &ApplyInitialForces()
   /Niagara/Modules/Update/Camera/ViewMask(Border Percent Max=5, Border Percent Min=0)
   &KillParticles(Kill Particles=&SetBoolByFloatComparison(A=0.2, B=$Output.ViewMask.ViewMask)) {module=KillParticles004}
   /Niagara/Modules/Update/Utility/SampleGBuffer(
    Buffer To Query:ENiagaraGBufferList_v2=::NewEnumerator7
    Kill Beyond Depth=true
    Max Z Depth=5000
    Snap To Depth:ENiagaraGBufferSnapToDepth=::NewEnumerator1
    Write To Color=false
   ) {module=SampleGBuffer004}
   &SampleGBuffer(Buffer To Query:ENiagaraGBufferList_v2=::NewEnumerator8)
   &SampleGBuffer(Buffer To Query:ENiagaraGBufferList_v2=::NewEnumerator12, Write To Color=false) {module=SampleGBuffer003}
   &KillParticles(Kill Particles=&SetBoolByFloatComparison(A=&DotProduct(Vector A=<0 0 1>, Vector B=$Output.SampleGBuffer003.WorldNormal), B=0.9)) {module=KillParticles002}
   &KillParticles(Kill Particles=&SetBoolByFloatComparison(A=$Output.SampleGBuffer004.Roughness, B=0.3))
   &KillParticles(Kill Particles=&SetBoolByFloatComparison(
     A=&DotProduct(Vector A=&MakeVectorFromLinearColorRGB(Color=$Particles.Color), Vector B=<0.2126 0.7152 0.0722>)
     B=0.001
     Comparison Type=::NewEnumerator6
    )) {enabled=false, module=KillParticles003}
   &SampleGBuffer(Buffer To Query:ENiagaraGBufferList_v2=::NewEnumerator11, Write To Color=false) {enabled=false, module=SampleGBuffer002}
   &KillParticles(Kill Particles=&SetBoolByFloatComparison(A=$Output.SampleGBuffer002.Specular, B=0.01, Comparison Type=::NewEnumerator6)) {enabled=false, module=KillParticles001}
  ]
  particleUpdateScript [
   &ParticleState() {module=UpdateAge}
   /Niagara/Modules/Update/Camera/MaintainInCameraParticleScale()
   &ScaleSpriteSize(Uniform Curve Scale=&Multiply_Float(
     A=4
     B=/Niagara/DynamicInputs/Camera/CameraFade(
      Camera Query:NiagaraDataInterfaceCamera=DataInterface()
      Fade Offset:NiagaraFloat=200
      Fade Distance:NiagaraFloat=2000
      Use Owner Position:NiagaraBool=true
     )
    ))
   &ScaleColor(
    Scale Alpha=&FloatFromCurve(FloatCurve=DataInterface() {propertyValues {Curve {keys [
         <at -0.0007 -0.0138 RCIM_Cubic>
         {arriveTangent=2.6721, interpMode=RCIM_Cubic, leaveTangent=2.6721, tangentMode=RCTM_User, time=0.3127, value=0.3952}
         {arriveTangent=-0.5712, interpMode=RCIM_Cubic, leaveTangent=-0.5712, tangentMode=RCTM_User, time=0.5424, value=1.0086}
         <at 1.0011 0.002 RCIM_Cubic>
        ]}}})
    ScaleA=true
    ScaleRGB=false
   )
   /SetVariables(Particles.RandomGlintVector:Vector3f=&RotateVectorByQuaternion(
     Vector To Rotate=&RandomVector(Vector Scale=1)
     Quaternion=&MakeQuatFromAxisAngle(
      Axis=&RandomVector(Vector Scale=1)
      AngleInDegrees=&Multiply_Float(A=$Particles.NormalizedAge, B=&RandomRangeFloat(Minimum=2000, Maximum=3000))
     )
    )) {isSetParameters=true}
   &WindForce(
    Scale Min / Max:Vector2f=<5 1>
    Speed Min / Max:Vector2f=<50 250>
    Turbulence Frequency Mode:ENiagaraWindTurbulenceFrequencyMode=::NewEnumerator1
    Turbulence Scale Mode:ENiagaraWindTurbulenceContributionMode=::NewEnumerator1
    Wind Speed=<2 0 0>
   )
   &AerodynamicDrag(
    Aerodynamic Drag=0.25
    Drag Shape Curve=DataInterface() {propertyValues {Curve {keys [
        {interpMode=RCIM_Cubic, tangentMode=RCTM_User, time=0, value=0.1}
        {arriveTangent=2.0799, interpMode=RCIM_Cubic, leaveTangent=2.0799, tangentMode=RCTM_User, time=1, value=1}
       ]}}}
    Lift Contribution=0
    Lift Shape Curve=DataInterface() {propertyValues {Curve {keys [
        {interpMode=RCIM_Cubic, tangentMode=RCTM_User, time=0, value=0}
        {interpMode=RCIM_Cubic, tangentMode=RCTM_User, time=0.5, value=1}
        {tangentMode=RCTM_User, time=1, value=0}
       ]}}}
    Use Linear Drag=true
    Use Rotational Drag=true
   )
   &SpriteRotationRate(Rotation Rate=&UniformRangedFloat(Minimum=-15, Maximum=25))
   &ScaleSpriteSize(Scale Sprite Size Mode=::NewEnumerator0, Uniform Scale Factor=2) {enabled=false, module=ScaleSpriteSize001}
   &ScaleColor(Scale RGB=<100 1 1>, ScaleA=true, ScaleRGB=true) {enabled=false, module=ScaleColor001}
   /SetVariables(
    Particles.Linear Color=&Multiply_LinearColorByFloat(Linear Color=$Particles.Color, Float=5)
    Particles.Vector 2D:Vector2f=&Multiply_Vector2D_ByFloat(Float=0.25, Vector2D=$Particles.SpriteSize)
   ) {isSetParameters=true}
   &SolveForcesAndVelocity(Clamp Velocity=false, Limit Acceleration=false)
   &AlignSpriteToMeshOrientation(SetAlignment=true, SetFacing=true)
   &CameraOffset(Camera Offset Amount=5)
   &KillParticlesInVolume(Box Size:Vector3f=<1350 1900 1300>, Invert Volume=true, Kill Shape=::NewEnumerator1)
  ]
  rendererClasses [&NiagaraSpriteRendererProperties]
  renderers [
   {rendererClass=&NiagaraSpriteRendererProperties}
   {rendererClass=&NiagaraSpriteRendererProperties, rendererIndex=1}
  ]
 }}
summary {
 systemName=NS_IceBiome_InteriorSparkles
 emitters [{bEnabled=false, emitterName=InteriorSparkles, simTarget=GPUComputeSim, rendererClasses [&NiagaraSpriteRendererProperties]}]
 userVariables []
}
