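# Niagara system for sparse snowflakes drifting down at an opening, apparently into the ice biome interior.
# One GPU emitter, SnowflakesAir, spawns long-lived flakes over an owner-scaled plane, thinning spawn count with distance and culling when far.
# Gravity, masked panning curl noise and drag drive the motion, while an alpha curve fades flakes in and out.
# It carries sprite and mesh renderers; wind, cone velocity and random-glint material parameters are present but disabled.
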NiagaraSystem /Game/Effects/Fort_Effects/Effects/Environment/Hera/IceBiome/IceBiomeInterior/NS_IceBiome_Opening_Snow
dependencies {
 usedDataInterfaces [&NiagaraDataInterfaceCurve]
 usedDynamicInputs [&Vector2DFromFloat]
 usedModules [&SystemState]
}
emitters {SnowflakesAir {
  simTarget=GPUComputeSim
  emitterUpdateScript [
   &EmitterState(
    Enable Distance Culling=true
    MaxDistance=10000
    Module.bApplySpawnCountScaleByDistance=true
    Module.bUseMaxDistance=true
    Module.bUseMinDistance=true
    Scalability Mode=::NewEnumerator1
    Scale Spawn Count=true
    Spawn Count Scale By Distance=DataInterface() {propertyValues {Curve {keys [<at 0 1>,<at 0.5 1>,<at 1 0>]}}}
   )
   &SpawnRate(SpawnRate=2, Use Spawn Probability=false) {module=SpawnRate001}
  ]
  particleSpawnScript [
   &InitializeParticle(
    Lifetime=&UniformRangedFloat(Minimum=10, Maximum=15)
    Mass Mode=::NewEnumerator3
    Sprite Rotation Mode=::NewEnumerator2
    Sprite Size Mode=::NewEnumerator1
    Uniform Sprite Size Max=7
    Uniform Sprite Size Min=6
    UsePositionOffset=false
    Write Material Random=true
   ) {module=InitializeParticle001}
   &AddVelocityInCone(Cone Angle=10, Cone Axis=<0 0 -1>, Cone Axis Coordinate Space=::World, Module.Use Velocity Falloff On Cone Axis=false, Velocity Strength=300) {enabled=false}
   &ShapeLocation(
    Box / Plane Mode:ENiagara_BoxPlaneMode=::NewEnumerator1
    MaskSpawnGroup=false
    Override Seed=false
    Plane Size:Vector2f=<800 2000>
    Shape Primitive=::NewEnumerator2
    UseOwnerScale=true
   )
   &GravityForce(Gravity=<0 0 -45>) {module=GravityForce001}
   &ApplyInitialForces(Apply Force to Position=true, Apply Force to Velocity=false, Apply Rotational Force to Rotational Velocity=false)
  ]
  particleUpdateScript [
   &ParticleState()
   &ScaleColor(
    Scale Alpha=&FloatFromCurve(
     FloatCurve=DataInterface() {propertyValues {Curve {keys [
         <at -0.0007 -0.0138 RCIM_Cubic>
         {arriveTangent=-0.0453, interpMode=RCIM_Cubic, leaveTangent=-0.0453, tangentMode=RCTM_User, time=0.1623, value=1.0362}
         {arriveTangent=-0.2452, interpMode=RCIM_Cubic, leaveTangent=-0.2452, tangentMode=RCTM_User, time=0.5424, value=1.0086}
         <at 1.0011 0.002 RCIM_Cubic>
        ]}}}
     Scale Curve=0.3
    )
    ScaleA=true
    ScaleRGB=true
   )
   &GravityForce(Gravity=<0 0 -68.4182>)
   /Niagara/Modules/Update/Forces/CurlNoiseForce(
    Curl Noise Cone Mask Angle=120
    Curl Noise Cone Mask Falloff Angle=100
    Mask Curl Noise=true
    Noise Frequency=384.4039
    Noise Strength=500
    Pan Noise=true
    Pan Noise Field=<0 1 2>
   )
   &WindForce(
    Friction Mode:ENiagaraWindFrictionMode=::NewEnumerator0
    Surface Alignment Mode:ENiagaraWindCombingMode=::NewEnumerator0
    Turbulence Frequency Mode:ENiagaraWindTurbulenceFrequencyMode=::NewEnumerator1
    Wind Speed=<-5 0 0>
   ) {enabled=false}
   &Drag(Ignore Mass=false, Use Linear Drag=true, Use Rotational Drag=true)
   &SolveForcesAndVelocity(Clamp Velocity=false, Limit Acceleration=false)
   /SetVariables(Particles.RandomGlintVector:Vector3f=&RotateVectorByQuaternion(
     Vector To Rotate=&RandomVector(Vector Scale=1)
     Quaternion=&MakeQuatFromAxisAngle(
      Axis=&RandomVector(Vector Scale=1)
      AngleInDegrees=&Multiply_Float(A=$Particles.NormalizedAge, B=&RandomRangeFloat(Minimum=2000, Maximum=3000))
     )
    )) {enabled=false, isSetParameters=true}
   &DynamicMaterialParameters(
    Index 0 Param 1=&RandomRangeFloat(Minimum=0, Maximum=1)
    Index 0 Param 2=&MakeFloatFromVector(Vector=$Particles.RandomGlintVector, Channel=::NewEnumerator0)
    Index 0 Param 3=&MakeFloatFromVector(Vector=$Particles.RandomGlintVector, Channel=::NewEnumerator1)
    Index 0 Param 4=&MakeFloatFromVector(Vector=$Particles.RandomGlintVector, Channel=::NewEnumerator2)
    Module.Param0WriteEnabled=true
    Module.Param1WriteEnabled=true
    Module.Param2WriteEnabled=true
    Module.Param3WriteEnabled=true
   ) {enabled=false}
  ]
  rendererClasses [&NiagaraSpriteRendererProperties, &NiagaraMeshRendererProperties]
  renderers [
   {rendererClass=&NiagaraSpriteRendererProperties}
   {rendererClass=&NiagaraMeshRendererProperties, rendererIndex=1}
  ]
 }}
summary {
 systemName=NS_IceBiome_Opening_Snow
 emitters [{
   emitterName=SnowflakesAir
   simTarget=GPUComputeSim
   rendererClasses [&NiagaraSpriteRendererProperties, &NiagaraMeshRendererProperties]
  }]
 userVariables []
}
