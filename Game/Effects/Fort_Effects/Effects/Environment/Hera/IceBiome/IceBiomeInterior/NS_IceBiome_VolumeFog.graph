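# Niagara system for drifting ambient volume fog in the Hera ice biome interior.
# A GPU FogAmbient emitter (sprite and mesh renderers) pairs with a bluer, sprite-only FogAmbientUnlit using SubUV animation.
# Both spawn in a flat box scaled by User.SpawnVolumeScale, drift along +Y, and are trimmed by kill volumes.
# Lifetime curves and camera fade shape opacity and sprite size, while distance culling scales spawn count; User.SpawnRateScale is declared but unreferenced.
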NiagaraSystem /Game/Effects/Fort_Effects/Effects/Environment/Hera/IceBiome/IceBiomeInterior/NS_IceBiome_VolumeFog
dependencies {
 usedDataInterfaces [
  &NiagaraDataInterfaceCurve
  &NiagaraDataInterfaceCamera
  &NiagaraDataInterfaceSpriteRendererInfo
 ]
 usedDynamicInputs []
 usedModules [&SystemState]
}
emitters {
 FogAmbient {
  simTarget=GPUComputeSim
  emitterUpdateScript [
   &EmitterState(
    Enable Distance Culling=true
    MaxDistance=8000
    Module.bApplySpawnCountScaleByDistance=true
    Module.bUseMaxDistance=true
    Module.bUseMinDistance=false
    Scalability Mode=::NewEnumerator1
    Scale Spawn Count=true
    Spawn Count Scale By Distance=DataInterface() {propertyValues {Curve {keys [<at 0 0>,<at 0.1 1>,<at 0.75 1>,<at 1 0>]}}}
   )
   &SpawnRate(SpawnRate=8, Use Spawn Probability=false)
  ]
  particleSpawnScript [
   &InitializeParticle(
    Color=<rgba 0.8522 0.9 1 1>
    Lifetime Max=6
    Lifetime Min=5
    Lifetime Mode=::NewEnumerator1
    Mesh Scale Mode=::NewEnumerator1
    Mesh Uniform Scale Max=12
    Mesh Uniform Scale Min=6
    Sprite Rotation Angle=&RandomRangeFloat(Minimum=-25, Maximum=25)
    Sprite Rotation Mode=::NewEnumerator0
    Sprite Size Mode=::NewEnumerator1
    Uniform Sprite Size Max=800
    Uniform Sprite Size Min=400
    UsePositionOffset=false
    Write Material Random=true
   )
   &ShapeLocation(
    Box Size=&Multiply_VectorByFloat(Vector=<3500 2000 1>, Float=$User.SpawnVolumeScale)
    MaskSpawnGroup=false
    Override Seed=false
    Shape Primitive=::NewEnumerator2
    Surface Only=false
    UseOwnerScale=true
   )
   &KillParticlesInVolume(Box Size:Vector3f=<1000 5000 1000>, Kill Shape=::NewEnumerator1) {module=KillParticlesInVolume001}
  ]
  particleUpdateScript [
   &ParticleState()
   &ScaleMeshSize(Scale Factor=&VectorFromFloat(Value=&FloatFromCurve(
      FloatCurve=DataInterface() {propertyValues {Curve {defaultValue=1, keys [{tangentMode=RCTM_User, time=0, value=0.5}, {tangentMode=RCTM_User, time=1, value=1}]}}}
      Scale Curve=0.5
     )))
   &KillParticlesInVolume(Box Size:Vector3f=<40000 500 1000>, Kill Shape=::NewEnumerator1, Origin Offset:Vector3f=<0 2500 0>)
   &KillParticlesInVolume(Box Size:Vector3f=<500 500 1000>, Kill Shape=::NewEnumerator1, Origin Offset:Vector3f=<-1000 2000 0>) {module=KillParticlesInVolume002}
   &KillParticlesInVolume(Box Size:Vector3f=<500 500 1000>, Kill Shape=::NewEnumerator1, Origin Offset:Vector3f=<1000 2000 0>) {module=KillParticlesInVolume003}
   &ScaleSpriteSize(Scale Sprite Size Mode=::NewEnumerator0, Uniform Scale Factor=0.2) {enabled=false}
   &ScaleSpriteSize(Uniform Curve Sprite Scale=DataInterface() {propertyValues {Curve {keys [{arriveTangent=1.3931, time=0, value=0.75}, <at 1 1.5>]}}}) {module=ScaleSpriteSize001}
   &ScaleSpriteSize(Uniform Curve Index=/Niagara/DynamicInputs/Camera/CameraFade(
     Camera Query:NiagaraDataInterfaceCamera=DataInterface()
     Fade Offset:NiagaraFloat=200
     Fade Distance:NiagaraFloat=500
     Use Owner Position:NiagaraBool=true
    )) {module=ScaleSpriteSize002}
   &ScaleColor(
    Scale Alpha=&FloatFromCurve(FloatCurve=DataInterface() {propertyValues {Curve {keys [
         {interpMode=RCIM_Cubic, tangentMode=RCTM_User, time=0, value=0}
         {interpMode=RCIM_Cubic, tangentMode=RCTM_User, time=0.5, value=1}
         {tangentMode=RCTM_User, time=1, value=0}
        ]}}})
    ScaleA=true
    ScaleRGB=false
   )
   &ScaleColor(
    Scale Alpha=&FloatFromCurve(
     FloatCurve=DataInterface() {propertyValues {Curve {defaultValue=0, keys [{tangentMode=RCTM_User, time=0, value=0}, {tangentMode=RCTM_User, time=1, value=1}]}}}
     CurveIndex=&CameraFade(
      Camera Query:NiagaraDataInterfaceCamera=DataInterface()
      Fade Offset:NiagaraFloat=300
      Fade Distance:NiagaraFloat=800
      Use Owner Position:NiagaraBool=true
     )
    )
    ScaleA=true
    ScaleRGB=false
   ) {module=ScaleColor001}
   &ScaleColor(Scale Alpha=0.2, ScaleA=true, ScaleRGB=false) {module=ScaleColor003}
   &AccelerationForce(Acceleration=&RandomRangeVector(Minimum=<0 40 0>, Maximum=<0 67.5 1>))
   &Drag(Drag=2, Use Linear Drag=true, Use Rotational Drag=true)
   &GravityForce(Gravity=&Multiply_VectorByFloat(Vector=<0 0 -1>, Float=1))
   &SolveForcesAndVelocity(Clamp Velocity=false, Limit Acceleration=false)
   &ScaleColor(Scale Alpha=5, Scale RGB=<100 1 1>, ScaleA=true, ScaleRGB=true) {enabled=false, module=ScaleColor002}
  ]
  rendererClasses [&NiagaraSpriteRendererProperties, &NiagaraMeshRendererProperties]
  renderers [
   {rendererClass=&NiagaraSpriteRendererProperties}
   {rendererClass=&NiagaraMeshRendererProperties, rendererIndex=1}
   {rendererClass=&NiagaraSpriteRendererProperties, rendererIndex=2}
  ]
 }
 FogAmbientUnlit {
  emitterUpdateScript [
   &EmitterState(
    Enable Distance Culling=true
    MaxDistance=8000
    Module.bApplySpawnCountScaleByDistance=true
    Module.bUseMaxDistance=true
    Module.bUseMinDistance=false
    Scalability Mode=::NewEnumerator1
    Scale Spawn Count=true
    Spawn Count Scale By Distance=DataInterface() {propertyValues {Curve {keys [<at 0 0>,<at 0.1 1>,<at 0.75 1>,<at 1 0>]}}}
   )
   &SpawnRate(SpawnRate=8, Use Spawn Probability=false)
  ]
  particleSpawnScript [
   &InitializeParticle(
    Color=<rgba 0.1217 0.4057 1 1>
    Lifetime Max=6
    Lifetime Min=5
    Lifetime Mode=::NewEnumerator1
    Mesh Scale Mode=::NewEnumerator1
    Mesh Uniform Scale Max=12
    Mesh Uniform Scale Min=6
    Sprite Rotation Angle=&RandomRangeFloat(Minimum=-25, Maximum=25)
    Sprite Rotation Mode=::NewEnumerator0
    Sprite Size Mode=::NewEnumerator1
    Uniform Sprite Size Max=800
    Uniform Sprite Size Min=400
    UsePositionOffset=false
    Write Material Random=true
   )
   &ShapeLocation(
    Box Size=&Multiply_VectorByFloat(Vector=<3500 2000 1>, Float=$User.SpawnVolumeScale)
    MaskSpawnGroup=false
    Override Seed=false
    Shape Primitive=::NewEnumerator2
    Surface Only=false
    UseOwnerScale=true
   )
   &KillParticlesInVolume(Box Size:Vector3f=<1000 5000 1000>, Kill Shape=::NewEnumerator1) {module=KillParticlesInVolume001}
  ]
  particleUpdateScript [
   &ParticleState()
   &ScaleMeshSize(Scale Factor=&VectorFromFloat(Value=&FloatFromCurve(
      FloatCurve=DataInterface() {propertyValues {Curve {defaultValue=1, keys [{tangentMode=RCTM_User, time=0, value=0.5}, {tangentMode=RCTM_User, time=1, value=1}]}}}
      Scale Curve=0.5
     )))
   &KillParticlesInVolume(Box Size:Vector3f=<40000 500 1000>, Kill Shape=::NewEnumerator1, Origin Offset:Vector3f=<0 2500 0>)
   &KillParticlesInVolume(Box Size:Vector3f=<500 500 1000>, Kill Shape=::NewEnumerator1, Origin Offset:Vector3f=<-1000 2000 0>) {module=KillParticlesInVolume002}
   &KillParticlesInVolume(Box Size:Vector3f=<500 500 1000>, Kill Shape=::NewEnumerator1, Origin Offset:Vector3f=<1000 2000 0>) {module=KillParticlesInVolume003}
   &ScaleSpriteSize(Scale Sprite Size Mode=::NewEnumerator0, Uniform Scale Factor=0.2) {enabled=false}
   &ScaleSpriteSize(Uniform Curve Sprite Scale=DataInterface() {propertyValues {Curve {keys [{arriveTangent=1.3931, time=0, value=0.75}, <at 1 1.5>]}}}) {module=ScaleSpriteSize001}
   &ScaleSpriteSize(Uniform Curve Index=&CameraFade(
     Camera Query:NiagaraDataInterfaceCamera=DataInterface()
     Fade Offset:NiagaraFloat=200
     Fade Distance:NiagaraFloat=500
     Use Owner Position:NiagaraBool=true
    )) {module=ScaleSpriteSize002}
   &ScaleColor(
    Scale Alpha=&FloatFromCurve(FloatCurve=DataInterface() {propertyValues {Curve {keys [
         {interpMode=RCIM_Cubic, tangentMode=RCTM_User, time=0, value=0}
         {interpMode=RCIM_Cubic, tangentMode=RCTM_User, time=0.5, value=1}
         {tangentMode=RCTM_User, time=1, value=0}
        ]}}})
    ScaleA=true
    ScaleRGB=false
   )
   &ScaleColor(
    Scale Alpha=&FloatFromCurve(
     FloatCurve=DataInterface() {propertyValues {Curve {defaultValue=0, keys [{tangentMode=RCTM_User, time=0, value=0}, {tangentMode=RCTM_User, time=1, value=1}]}}}
     CurveIndex=&CameraFade(
      Camera Query:NiagaraDataInterfaceCamera=DataInterface()
      Fade Offset:NiagaraFloat=300
      Fade Distance:NiagaraFloat=800
      Use Owner Position:NiagaraBool=true
     )
    )
    ScaleA=true
    ScaleRGB=false
   ) {module=ScaleColor001}
   &ScaleColor(Scale Alpha=0.3, ScaleA=true, ScaleRGB=false) {module=ScaleColor003}
   &AccelerationForce(Acceleration=&RandomRangeVector(Minimum=<0 40 0>, Maximum=<0 67.5 1>))
   &Drag(Drag=2, Use Linear Drag=true, Use Rotational Drag=true)
   &GravityForce(Gravity=&Multiply_VectorByFloat(Vector=<0 0 -1>, Float=1))
   &SolveForcesAndVelocity(Clamp Velocity=false, Limit Acceleration=false)
   &DynamicMaterialParameters(
    Index 0 Param 2=1
    Index 0 Param 3=10
    Module.Index0Float0Write=true
    Module.Index0Float1Write=true
    Module.Index0Float2Write=true
    Module.Index0Float3Write=true
   )
   &ScaleColor(Scale Alpha=5, Scale RGB=<100 1 1>, ScaleA=true, ScaleRGB=true) {enabled=false, module=ScaleColor002}
   &SubUVAnimation(
    "Sprite Renderer "=DataInterface() {propertyValues {SpriteRenderer=&self:FogAmbientUnlit_0.NiagaraSpriteRendererProperties_0}}
    UseEndFrame=false
    UseStartFrame=false
   )
  ]
  rendererClasses [&NiagaraSpriteRendererProperties]
  renderers [{rendererClass=&NiagaraSpriteRendererProperties}]
 }
}
summary {
 systemName=NS_IceBiome_VolumeFog
 emitters [
  {
   emitterName=FogAmbient
   simTarget=GPUComputeSim
   rendererClasses [&NiagaraSpriteRendererProperties, &NiagaraMeshRendererProperties]
  }
  {emitterName=FogAmbientUnlit, rendererClasses [&NiagaraSpriteRendererProperties]}
 ]
 userVariables [
  User.SpawnRateScale {type=NiagaraFloat, defaultValue=1, description=""}
  User.SpawnVolumeScale {type=NiagaraFloat, defaultValue=1, description=""}
 ]
}
